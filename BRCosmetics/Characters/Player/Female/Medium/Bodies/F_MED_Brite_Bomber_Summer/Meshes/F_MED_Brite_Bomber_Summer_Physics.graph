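# Physics asset for the female Brite Bomber Summer character body, covering a glasses accessory and the hip area around it.
# Capsule bodies sit on the pelvis, right thigh and two glasses bones.
# The glasses chain runs from pelvis through a tiny parent body to the glasses body, which has tight swing and twist limits.
# Constraints from the glasses parent and right thigh to the pelvis leave rotation free.
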PhysicsAsset /BRCosmetics/Characters/Player/Female/Medium/Bodies/F_MED_Brite_Bomber_Summer/Meshes/F_MED_Brite_Bomber_Summer_Physics
bodies {
 dyn_glasses_01 [{shapeType=Capsule, length=7.3024, radius=3.9731, center=<9.5238 -1.1254 0.0002>, rotation=<rot 90 0 0>}]
 dyn_parent_glasses [{shapeType=Capsule, length=0.505, radius=0.505, rotation=<rot 0 -90 0>}]
 pelvis [{shapeType=Capsule, length=5.2521, radius=5.5522, center=<0.1164 -1.1311 0.1887>, rotation=<rot -4.0634 -90 -1.6374>}]
 thigh_r [{shapeType=Capsule, length=12.1028, radius=8.4975, center=<14.202 -0.7038 0.5817>, rotation=<rot 0.0227 -91.8955 -89.5714>}]
}
constraints [
 {
  bone1Name=dyn_parent_glasses
  bone2Name=pelvis
  swing1LimitDegrees=45
  swing1Motion=Free
  swing2LimitDegrees=45
  swing2Motion=Free
  twistLimitDegrees=45
  twistMotion=Free
 }
 {bone1Name=dyn_glasses_01, bone2Name=dyn_parent_glasses, swing1LimitDegrees=12, swing2LimitDegrees=20, twistLimitDegrees=5}
 {
  bone1Name=thigh_r
  bone2Name=pelvis
  swing1LimitDegrees=45
  swing1Motion=Free
  swing2LimitDegrees=45
  swing2Motion=Free
  twistLimitDegrees=45
  twistMotion=Free
 }
]
