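# Animation Blueprint for the F_MED_Brite_Bomber_Summer female body cosmetic, layering secondary physics and joint correctives onto the pose copied from its attached parent mesh.
# A RigidBody node simulates the override physics asset in root-bone space with custom gravity and damped component motion.
# Additive bone modifiers then adjust the left glute and both knee dyn bones.
# Each update, left thigh and both knee bend yaws are remapped into alphas driving those correctives.
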AnimBlueprint /BRCosmetics/Characters/Player/Female/Medium/Bodies/F_MED_Brite_Bomber_Summer/Meshes/F_MED_Brite_Bomber_Summer_AnimBP
target_skeleton=&/F_MED_Brite_Bomber_Summer_Skeleton
graphs [
 AnimGraph {nodes [
   n1 = RigidBody(ComponentPose=Local To Component(LocalPose=Copy Pose From Mesh() {bUseAttachedParent=true})) {
    bClampLinearTranslationLimitToRefPose=true
    bOverrideWorldGravity=true
    overridePhysicsAsset=&/F_MED_Brite_Bomber_Summer_Physics
    simulationSpace=BaseBoneSpace
    alphaBoolBlend {blendOption=Linear}
    baseBoneRef {boneName=root}
    componentAppliedLinearAccClamp=<1100 1100 1100>
    componentLinearAccScale=<0.3 0.3 0.25>
    componentLinearVelScale=<0.3 0.3 0.25>
    overrideWorldGravity {z=-980}
   }
   n2 = "Transform (Modify) Bone - Bone: dyn_glute_l"(ComponentPose=@n1, Alpha=Get thigh_rot_l(), Translation="-2.750,-0.150,3.412", Rotation="0.000,-0.001,-58", Scale="1.35,1.35,1.35") {
    rotationMode=BMM_Additive
    rotationSpace=BCS_BoneSpace
    scaleMode=BMM_Additive
    scaleSpace=BCS_BoneSpace
    translationMode=BMM_Additive
    translationSpace=BCS_BoneSpace
    alphaBoolBlend {blendOption=Linear}
    boneToModify {boneName=dyn_glute_l}
    rotation {roll=-58, yaw=-0.001}
    scale=<1.35 1.35 1.35>
    translation=<-2.75 -0.15 3.412>
   }
   n3 = "Transform (Modify) Bone - Bone: dyn_knee_r"(ComponentPose=@n2, Alpha=Get calf_rot_r() -> (calf_rot_r:double), Translation="0.000,-5,3", Rotation="0,0,0", Scale="1.35,1.35,1.35") {
    rotationMode=BMM_Additive
    rotationSpace=BCS_BoneSpace
    scaleMode=BMM_Additive
    scaleSpace=BCS_BoneSpace
    translationMode=BMM_Additive
    translationSpace=BCS_BoneSpace
    alphaBoolBlend {blendOption=Linear}
    boneToModify {boneName=dyn_knee_r}
    scale=<1.35 1.35 1.35>
    translation {y=-5, z=3}
   }
   n4 = "Transform (Modify) Bone - Bone: dyn_knee_l"(ComponentPose=@n3, Alpha=Get calf_rot_l() -> (calf_rot_l:double), Translation="0.000,5,-3", Rotation="0,0,0", Scale="1.35,1.35,1.35") {
    rotationMode=BMM_Additive
    rotationSpace=BCS_BoneSpace
    scaleMode=BMM_Additive
    scaleSpace=BCS_BoneSpace
    translationMode=BMM_Additive
    translationSpace=BCS_BoneSpace
    alphaBoolBlend {blendOption=Linear}
    boneToModify {boneName=dyn_knee_l}
    scale=<1.35 1.35 1.35>
    translation {y=5, z=-3}
   }
   Output Pose(Result=Component To Local(ComponentPose=@n4)) {layerGroup=None}
  ]}
 EventGraph {nodes [
   Event BlueprintUpdateAnimation()
   > n1 = Sequence()
   n2 = BreakTransform(InTransform=GetDeltaTransformFromRefPose(self=GetOwnerSkeletalMeshComponent(), BoneName=thigh_l, BaseName=pelvis))
   > Set thigh_rot_l(thigh_rot_l=MapRangeClamped(Value=BreakRotator(InRot=@n2.Rotation).Yaw, InRangeA=-20, InRangeB=-100))
   n3 = GetOwnerSkeletalMeshComponent()
   n4 = BreakTransform(InTransform=GetDeltaTransformFromRefPose(self=@n3, BoneName=calf_l, BaseName=thigh_l))
   Set calf_rot_l(execute=@n1.then_1, calf_rot_l:double=MapRangeClamped(Value=BreakRotator(InRot=@n4.Rotation).Yaw, InRangeA=0, InRangeB=115)) -> (Output_Get:double)
   n5 = BreakTransform(InTransform=GetDeltaTransformFromRefPose(self=@n3, BoneName=thigh_r, BaseName=calf_r))
   > Set calf_rot_r(calf_rot_r:double=MapRangeClamped(Value=BreakRotator(InRot=@n5.Rotation).Yaw, InRangeA=0, InRangeB=-115)) -> (Output_Get:double)
  ]}
]
variables [
 calf_rot_l {type=number}
 calf_rot_r {type=number}
 thigh_rot_l {type=number}
]
