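# Physics asset for the female Brite Bomber Summer body that, by its name, serves as the collision volume for capes.
# Capsules wrap the torso, neck, limbs, hands, feet and two dynamic glasses bones, with a sphere on the head.
# Constraints mostly follow the skeleton with free swing and twist, so they impose no angular limits.
# The glasses parent links to the pelvis, and upperarm_r has no constraint to clavicle_r, unlike the left side.
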PhysicsAsset /BRCosmetics/Characters/Player/Female/Medium/Bodies/F_MED_Brite_Bomber_Summer/Meshes/F_MED_Brite_Bomber_Summer_CapeCollider
bodies {
 calf_l [{shapeType=Capsule, length=28.7102, radius=8.4182, center=<-27.1911 -4.1582 -0.9495>, rotation=<rot -0.2451 88.2824 -88.6099>}]
 calf_r [{shapeType=Capsule, length=28.71, radius=8.4182, center=<27.1867 4.0365 0.9493>, rotation=<rot -0.2451 88.2856 -88.6097>}]
 clavicle_l [{shapeType=Capsule, length=11, radius=4.5, shapeName=clavicle_l_capsule, center=<7.4606 0 0>, rotation=<rot 80 -0.0001 -0.0004>}]
 clavicle_r [{shapeType=Capsule, length=9.5, radius=4.5, shapeName=clavicle_r_capsule, center=<-5.9983 0 0>, rotation=<rot 80 -0.0005 -0.0008>}]
 dyn_glasses_01 [{shapeType=Capsule, length=10, radius=10.5, shapeName=dyn_glasses_01_capsule, center=<7.4837 -0.5265 0>, rotation=<rot 90 179.789 179.7889>}]
 dyn_parent_glasses [{shapeType=Capsule, length=0, radius=0.5, shapeName=dyn_parent_glasses_capsule}]
 foot_l [{shapeType=Capsule, length=15, radius=6, shapeName=foot_l_capsule, center=<-3.6781 6.0201 0>, rotation=<rot 0 -0.0002 90>}]
 foot_r [{shapeType=Capsule, length=15, radius=6.5, shapeName=foot_r_capsule, center=<4.5185 -6.9338 0>, rotation=<rot 0 -0.0002 90>}]
 hand_l [{shapeType=Capsule, length=8.8694, radius=6.0278, center=<7.7934 -2.5954 0.124>, rotation=<rot 0 85.4946 -70.874>}]
 hand_r [{shapeType=Capsule, length=8.8933, radius=6.0107, center=<-7.8038 2.5926 -0.0732>, rotation=<rot 0 85.3686 -71.9596>}]
 head [{shapeType=Sphere, radius=10.5, shapeName=head_sphere, center=<3.0491 0 0>}]
 lowerarm_l [{shapeType=Capsule, length=19, radius=5.5, shapeName=lowerarm_l_capsule, center=<9.3233 0 0>, rotation=<rot 90 -179.9999 179.9999>}]
 lowerarm_r [{shapeType=Capsule, length=19, radius=5.5, shapeName=lowerarm_r_capsule, center=<-6.1313 0 0>, rotation=<rot 90 -179.9406 -179.9408>}]
 neck_02 [{shapeType=Capsule, length=3, radius=3.5, shapeName=neck_02_capsule, center=<-3.0157 0 0>, rotation=<rot -90 179.4936 -179.4938>}]
 pelvis [{shapeType=Capsule, length=8.5, radius=17.5, shapeName=pelvis_capsule, center=<-3.576 0 0>}]
 spine_01 [{shapeType=Capsule, length=7.5, radius=9, shapeName=spine_01_capsule}]
 spine_02 [{shapeType=Capsule, length=8.5, radius=6.5, shapeName=spine_02_capsule}]
 spine_03 [{shapeType=Capsule, length=8.5, radius=6.5, shapeName=spine_03_capsule}]
 spine_04 [{shapeType=Capsule, length=4, radius=8, shapeName=spine_04_capsule}]
 spine_05 [{shapeType=Capsule, length=11, radius=10, shapeName=spine_05_capsule}]
 thigh_l [{shapeType=Capsule, length=27.901, radius=11.1579, center=<-25.2398 -0.9862 0.1583>, rotation=<rot 0 -83 -90>}]
 thigh_r [{shapeType=Capsule, length=27.8526, radius=10.9935, center=<25.2652 1.1599 0.1335>, rotation=<rot 0 -82.195 -90>}]
 upperarm_l [{shapeType=Capsule, length=15, radius=5.5, shapeName=upperarm_l_capsule, center=<11.3498 0.0001 0>, rotation=<rot 90 180 179.9997>}]
 upperarm_r [{shapeType=Capsule, length=15, radius=5.5, shapeName=upperarm_r_capsule, center=<-9.7412 -0.0001 0>, rotation=<rot 90 -6.7214 -6.7218>}]
}
constraints [
 {
  bone1Name=spine_01
  bone2Name=pelvis
  swing1LimitDegrees=45
  swing1Motion=Free
  swing2LimitDegrees=45
  swing2Motion=Free
  twistLimitDegrees=45
  twistMotion=Free
 }
 {
  bone1Name=spine_02
  bone2Name=spine_01
  swing1LimitDegrees=45
  swing1Motion=Free
  swing2LimitDegrees=45
  swing2Motion=Free
  twistLimitDegrees=45
  twistMotion=Free
 }
 {
  bone1Name=spine_03
  bone2Name=spine_02
  swing1LimitDegrees=45
  swing1Motion=Free
  swing2LimitDegrees=45
  swing2Motion=Free
  twistLimitDegrees=45
  twistMotion=Free
 }
 {
  bone1Name=spine_04
  bone2Name=spine_03
  swing1LimitDegrees=45
  swing1Motion=Free
  swing2LimitDegrees=45
  swing2Motion=Free
  twistLimitDegrees=45
  twistMotion=Free
 }
 {
  bone1Name=spine_05
  bone2Name=spine_04
  swing1LimitDegrees=45
  swing1Motion=Free
  swing2LimitDegrees=45
  swing2Motion=Free
  twistLimitDegrees=45
  twistMotion=Free
 }
 {
  bone1Name=clavicle_l
  bone2Name=spine_05
  swing1LimitDegrees=45
  swing1Motion=Free
  swing2LimitDegrees=45
  swing2Motion=Free
  twistLimitDegrees=45
  twistMotion=Free
 }
 {
  bone1Name=clavicle_r
  bone2Name=spine_05
  swing1LimitDegrees=45
  swing1Motion=Free
  swing2LimitDegrees=45
  swing2Motion=Free
  twistLimitDegrees=45
  twistMotion=Free
 }
 {
  bone1Name=upperarm_l
  bone2Name=clavicle_l
  swing1LimitDegrees=45
  swing1Motion=Free
  swing2LimitDegrees=45
  swing2Motion=Free
  twistLimitDegrees=45
  twistMotion=Free
 }
 {
  bone1Name=lowerarm_l
  bone2Name=upperarm_l
  swing1LimitDegrees=45
  swing1Motion=Free
  swing2LimitDegrees=45
  swing2Motion=Free
  twistLimitDegrees=45
  twistMotion=Free
 }
 {
  bone1Name=hand_l
  bone2Name=lowerarm_l
  swing1LimitDegrees=45
  swing1Motion=Free
  swing2LimitDegrees=45
  swing2Motion=Free
  twistLimitDegrees=45
  twistMotion=Free
 }
 {
  bone1Name=lowerarm_r
  bone2Name=upperarm_r
  swing1LimitDegrees=45
  swing1Motion=Free
  swing2LimitDegrees=45
  swing2Motion=Free
  twistLimitDegrees=45
  twistMotion=Free
 }
 {
  bone1Name=hand_r
  bone2Name=lowerarm_r
  swing1LimitDegrees=45
  swing1Motion=Free
  swing2LimitDegrees=45
  swing2Motion=Free
  twistLimitDegrees=45
  twistMotion=Free
 }
 {
  bone1Name=neck_02
  bone2Name=spine_05
  swing1LimitDegrees=45
  swing1Motion=Free
  swing2LimitDegrees=45
  swing2Motion=Free
  twistLimitDegrees=45
  twistMotion=Free
 }
 {
  bone1Name=head
  bone2Name=neck_02
  swing1LimitDegrees=45
  swing1Motion=Free
  swing2LimitDegrees=45
  swing2Motion=Free
  twistLimitDegrees=45
  twistMotion=Free
 }
 {
  bone1Name=thigh_l
  bone2Name=pelvis
  swing1LimitDegrees=45
  swing1Motion=Free
  swing2LimitDegrees=45
  swing2Motion=Free
  twistLimitDegrees=45
  twistMotion=Free
 }
 {
  bone1Name=thigh_r
  bone2Name=pelvis
  swing1LimitDegrees=45
  swing1Motion=Free
  swing2LimitDegrees=45
  swing2Motion=Free
  twistLimitDegrees=45
  twistMotion=Free
 }
 {
  bone1Name=calf_l
  bone2Name=thigh_l
  swing1LimitDegrees=45
  swing1Motion=Free
  swing2LimitDegrees=45
  swing2Motion=Free
  twistLimitDegrees=45
  twistMotion=Free
 }
 {
  bone1Name=foot_l
  bone2Name=calf_l
  swing1LimitDegrees=45
  swing1Motion=Free
  swing2LimitDegrees=45
  swing2Motion=Free
  twistLimitDegrees=45
  twistMotion=Free
 }
 {
  bone1Name=foot_r
  bone2Name=calf_r
  swing1LimitDegrees=45
  swing1Motion=Free
  swing2LimitDegrees=45
  swing2Motion=Free
  twistLimitDegrees=45
  twistMotion=Free
 }
 {
  bone1Name=calf_r
  bone2Name=thigh_r
  swing1LimitDegrees=45
  swing1Motion=Free
  swing2LimitDegrees=45
  swing2Motion=Free
  twistLimitDegrees=45
  twistMotion=Free
 }
 {
  bone1Name=dyn_parent_glasses
  bone2Name=pelvis
  swing1LimitDegrees=45
  swing1Motion=Free
  swing2LimitDegrees=45
  swing2Motion=Free
  twistLimitDegrees=45
  twistMotion=Free
 }
 {
  bone1Name=dyn_glasses_01
  bone2Name=dyn_parent_glasses
  swing1LimitDegrees=45
  swing1Motion=Free
  swing2LimitDegrees=45
  swing2Motion=Free
  twistLimitDegrees=45
  twistMotion=Free
 }
]
