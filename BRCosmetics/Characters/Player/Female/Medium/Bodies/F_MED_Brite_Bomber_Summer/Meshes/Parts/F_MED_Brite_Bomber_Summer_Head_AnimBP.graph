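# Head-part AnimBlueprint for the summer Brite Bomber skin, adding secondary motion to ponytail and left-side hair.
# It copies the parent mesh pose, then chains Anim Dynamics nodes, with spine_05 spherical limits keeping the ponytail off the body.
# Each update checks for a cape backbling, switching dyn_ponytail_03 between two simulations with wider limits for capes.
# With a cape, head rotation relative to spine_05 also drives an additive ponytail_MSTR roll before simulation.
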AnimBlueprint /BRCosmetics/Characters/Player/Female/Medium/Bodies/F_MED_Brite_Bomber_Summer/Meshes/Parts/F_MED_Brite_Bomber_Summer_Head_AnimBP
target_skeleton=&/F_MED_Brite_Bomber_Summer_Head_Skeleton
graphs [
 AnimGraph {nodes [
   n1 = "Transform (Modify) Bone - Bone: ponytail_MSTR"(
    ComponentPose=Local To Component(LocalPose=Copy Pose From Mesh() {bUseAttachedParent=true})
    Alpha=Get PonyTail Movement() -> (PonyTail Movement:double)
    Translation="0,0,0"
    Rotation="0,0,30"
    Scale="1,1,1"
   ) {
    rotationMode=BMM_Additive
    rotationSpace=BCS_BoneSpace
    alphaBoolBlend {blendOption=Linear}
    boneToModify {boneName=ponytail_MSTR}
    rotation {roll=30}
   }
   n2 = Anim Dynamics - Bone: dyn_ponytail_01(ComponentPose=@n1, Alpha=1, ExternalForce="0,0,0") {
    angularDampingOverride=1
    angularSpringConstant=300
    bAngularSpring=true
    bGravityOverrideInSimSpace=true
    bOverrideAngularDamping=true
    bUseSphericalLimits=true
    gravityScale=0.31
    simulationSpace=RootRelative
    alphaBoolBlend {blendOption=Linear}
    boundBone {boneName=dyn_ponytail_01}
    componentAppliedLinearAccClamp=<1000 1000 800>
    componentLinearAccScale=<0.5 0.5 0.25>
    componentLinearVelScale=<0.5 0.5 0.25>
    physicsBodyDefinitions [{
      collisionType=CustomSphere
      sphereCollisionRadius=3
      boundBone {boneName=dyn_ponytail_01}
      boxExtents=<20 20 15>
      constraintSetup {angularTargetAxis=AxisY, twistAxis=AxisZ, angularLimitsMax=<20 15 6>, angularLimitsMin=<-20 -15 -16>, angularTarget=<0 1 0>}
      localJointOffset=<0 0 -6>
     }]
    sphericalLimits [{limitRadius=55, drivingBone {boneName=spine_05}, sphereLocalOffset=<-35 20 0>}]
   }
   n3 = Anim Dynamics - Bone: dyn_ponytail_02(ComponentPose=@n2, Alpha=1, ExternalForce="0,0,0") {
    angularDampingOverride=1
    angularSpringConstant=1500
    bAngularSpring=true
    bGravityOverrideInSimSpace=true
    bOverrideAngularDamping=true
    bUseSphericalLimits=true
    gravityScale=0.32
    simulationSpace=RootRelative
    alphaBoolBlend {blendOption=Linear}
    boundBone {boneName=dyn_ponytail_02}
    componentAppliedLinearAccClamp=<1000 1000 1200>
    componentLinearAccScale=<0.75 0.75 0.95>
    componentLinearVelScale=<0.75 0.75 0.95>
    physicsBodyDefinitions [{
      collisionType=CustomSphere
      sphereCollisionRadius=2
      boundBone {boneName=dyn_ponytail_02}
      boxExtents=<15 25 25>
      constraintSetup {angularLimitsMax=<0 10 10>, angularLimitsMin=<0 -10 -60>, angularTarget=<0 0 0>}
      localJointOffset=<8 0 0>
     }]
    sphericalLimits [{limitRadius=55, drivingBone {boneName=spine_05}, sphereLocalOffset=<-10 45 0>}]
   }
   n4 = Anim Dynamics - Bone: dyn_ponytail_03(ComponentPose=@n3, Alpha=Get Is wearing cape alpha() -> (Is wearing cape alpha:double), ExternalForce="0,0,0") {
    angularDampingOverride=1
    angularSpringConstant=1500
    bAngularSpring=true
    bGravityOverrideInSimSpace=true
    bOverrideAngularDamping=true
    bUseSphericalLimits=true
    gravityScale=0.31
    simulationSpace=RootRelative
    alphaBoolBlend {blendOption=Linear}
    boundBone {boneName=dyn_ponytail_03}
    componentAppliedLinearAccClamp=<1000 1000 1200>
    componentLinearAccScale=<0.75 0.75 0.95>
    componentLinearVelScale=<0.75 0.75 0.95>
    physicsBodyDefinitions [{
      collisionType=CustomSphere
      sphereCollisionRadius=3.5
      boundBone {boneName=dyn_ponytail_03}
      boxExtents=<15 25 25>
      constraintSetup {angularLimitsMax=<0 25 8>, angularLimitsMin=<0 -25 -50>, angularTarget=<0 0 0>}
      localJointOffset=<5 0 0>
     }]
    sphericalLimits [
     {limitRadius=62, drivingBone {boneName=spine_05}, sphereLocalOffset=<-9 45 0>}
     {limitRadius=60, drivingBone {boneName=spine_05}, sphereLocalOffset=<-25 35 0>}
    ]
   }
   n5 = Anim Dynamics - Bone: dyn_ponytail_03(ComponentPose=@n4, Alpha=Get Is not wearing cape alpha() -> (Is not wearing cape alpha:double), ExternalForce="0,0,0") {
    angularDampingOverride=1
    angularSpringConstant=1500
    bAngularSpring=true
    bGravityOverrideInSimSpace=true
    bOverrideAngularDamping=true
    bUseSphericalLimits=true
    gravityScale=0.31
    simulationSpace=RootRelative
    alphaBoolBlend {blendOption=Linear}
    boundBone {boneName=dyn_ponytail_03}
    componentAppliedLinearAccClamp=<1000 1000 1200>
    componentLinearAccScale=<0.75 0.75 0.95>
    componentLinearVelScale=<0.75 0.75 0.95>
    physicsBodyDefinitions [{
      collisionType=CustomSphere
      sphereCollisionRadius=3.5
      boundBone {boneName=dyn_ponytail_03}
      boxExtents=<15 25 25>
      constraintSetup {angularLimitsMax=<0 25 8>, angularLimitsMin=<0 -25 -50>, angularTarget=<0 0 0>}
      localJointOffset=<5 0 0>
     }]
    sphericalLimits [
     {limitRadius=56, drivingBone {boneName=spine_05}, sphereLocalOffset=<-9 45 0>}
     {limitRadius=55, drivingBone {boneName=spine_05}, sphereLocalOffset=<-25 35 0>}
    ]
   }
   n6 = Anim Dynamics - Bone: dyn_ponytail_04(ComponentPose=@n5, Alpha=0.7, ExternalForce="0,0,0") {
    angularDampingOverride=1
    angularSpringConstant=500
    bAngularSpring=true
    bGravityOverrideInSimSpace=true
    bOverrideAngularDamping=true
    bUseSphericalLimits=true
    gravityScale=0.31
    simulationSpace=RootRelative
    alphaBoolBlend {blendOption=Linear}
    boundBone {boneName=dyn_ponytail_04}
    componentAppliedLinearAccClamp=<1000 1000 1200>
    componentLinearAccScale=<0.75 0.75 0.95>
    componentLinearVelScale=<0.75 0.75 0.95>
    physicsBodyDefinitions [{
      collisionType=CustomSphere
      sphereCollisionRadius=3
      boundBone {boneName=dyn_ponytail_04}
      boxExtents=<15 25 25>
      constraintSetup {angularTargetAxis=AxisZ, angularLimitsMax=<0 30 5>, angularLimitsMin=<0 -30 -40>, angularTarget=<0 0 1>}
      localJointOffset=<3 0 0>
     }]
    sphericalLimits [
     {limitRadius=55, drivingBone {boneName=spine_05}, sphereLocalOffset=<-10 44 0>}
     {limitRadius=50, drivingBone {boneName=spine_05}, sphereLocalOffset=<-20 35 0>}
    ]
   }
   n7 = Anim Dynamics - Bone: dyn_hair_01_l(ComponentPose=@n6, Alpha=1, ExternalForce="0,0,0") {
    angularDampingOverride=1
    angularSpringConstant=1500
    bAngularSpring=true
    bGravityOverrideInSimSpace=true
    bOverrideAngularDamping=true
    gravityScale=0.29
    simulationSpace=RootRelative
    alphaBoolBlend {blendOption=Linear}
    boundBone {boneName=dyn_hair_01_l}
    physicsBodyDefinitions [{
      sphereCollisionRadius=0
      boundBone {boneName=dyn_hair_01_l}
      boxExtents=<13 20 20>
      constraintSetup {angularLimitsMax=<0 10 3>, angularLimitsMin=<0 -35 -40>, angularTarget=<0 0 0>}
      localJointOffset=<10 0 0>
     }]
   }
   n8 = Anim Dynamics - Bone: dyn_hair_02_l(ComponentPose=@n7, Alpha=1, ExternalForce="0,0,0") {
    angularDampingOverride=1
    angularSpringConstant=1500
    bAngularSpring=true
    bGravityOverrideInSimSpace=true
    bOverrideAngularDamping=true
    gravityScale=0.29
    simulationSpace=RootRelative
    alphaBoolBlend {blendOption=Linear}
    boundBone {boneName=dyn_hair_02_l}
    physicsBodyDefinitions [{
      sphereCollisionRadius=0
      boundBone {boneName=dyn_hair_02_l}
      boxExtents=<13 20 20>
      constraintSetup {angularLimitsMax=<0 10 6>, angularLimitsMin=<0 -25 -40>, angularTarget=<0 0 0>}
      localJointOffset=<10 0 0>
     }]
   }
   n9 = Anim Dynamics - Bone: dyn_hair_03_l(ComponentPose=@n8, Alpha=1, ExternalForce="0,0,0") {
    angularDampingOverride=1
    angularSpringConstant=1500
    bAngularSpring=true
    bGravityOverrideInSimSpace=true
    bOverrideAngularDamping=true
    gravityScale=0.29
    simulationSpace=RootRelative
    alphaBoolBlend {blendOption=Linear}
    boundBone {boneName=dyn_hair_03_l}
    physicsBodyDefinitions [{
      sphereCollisionRadius=0
      boundBone {boneName=dyn_hair_03_l}
      boxExtents=<13 20 20>
      constraintSetup {angularLimitsMax=<0 15 5>, angularLimitsMin=<0 -25 -40>, angularTarget=<0 0 0>}
      localJointOffset=<10 0 0>
     }]
   }
   Output Pose(Result=Component To Local(ComponentPose=@n9)) {layerGroup=None}
  ]}
 EventGraph {nodes [
   Event BlueprintUpdateAnimation()
   > n1 = Sequence()
   > Set Is wearing cape(Is wearing cape:bool=CheckForBackbling().bIsBackblingACape) -> (Output_Get:bool)
   n2 = Get Is wearing cape() -> (Is wearing cape:bool)
   Set Is wearing cape alpha(execute=@n1.then_1, Is wearing cape alpha:double=SelectFloat(A=0.7, B=0, bPickA=@n2)) -> (Output_Get:double)
   > Set Is not wearing cape alpha(Is not wearing cape alpha:double=SelectFloat(A=0.7, B=0, bPickA=NOT Boolean(A=@n2))) -> (Output_Get:double)
   n3 = Get Skeletal Mesh() -> (Skeletal Mesh:Skeletal Mesh Component Object Reference)
   > Set PonyTail Movement(PonyTail Movement:double=SelectFloat(
     A=GetAlphaFromReferenceTransformDelta(SkeletalMeshComponent=@n3, BoneName=head, BaseName=spine_05, InputRange=(X=24,Y=-47), TwistAxis="1,0,0")
     B=0
     bPickA=@n2
    )) -> (Output_Get:double)
   Event BlueprintInitializeAnimation()
   > Set Skeletal Mesh(Skeletal Mesh:Skeletal Mesh Component Object Reference=GetOwnerSkeletalMeshComponent()) -> (Output_Get:Skeletal Mesh Component Object Reference)
  ]}
]
variables [
 Is not wearing cape alpha {type=number}
 Is wearing cape {type=boolean}
 Is wearing cape alpha {type=number}
 PonyTail Movement {type=number}
 Skeletal Mesh {type=SkeletalMeshComponent}
]
